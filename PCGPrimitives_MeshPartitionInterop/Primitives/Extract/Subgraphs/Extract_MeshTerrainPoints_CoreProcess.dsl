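; PCG subgraph that forms the core process for extracting mesh terrain points, turning spatial input into points that record their source positions.
; It filters the input to spatial data, passing only matches on to a To Point conversion.
; It then adds a vector SourcePositions attribute and copies each point's Position into it with a Set operation, preserving the original location as an attribute on the output.
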(graph Extract_MeshTerrainPoints_CoreProcess
  (bind FilterDataByType_0 (Filter%20Data%20By%20Type :TargetType (struct :Ids [(struct :Struct /Script/PCG.PCGDataTypeInfoSpatial)])))
  (bind ToPoint_2 (To%20Point))
  (bind AddAttribute_2 (Add%20Attribute :AttributeTypes (struct :Type Vector) :OutputTarget SourcePositions))
  (bind AttributeMathsOp_1 (Attribute%20Maths%20Op :InputSource1 $Position :InputSource2 $Position :Operation Set :OutputTarget SourcePositions))
  (connect ToPoint_2 Out AddAttribute_2 In)
  (connect AddAttribute_2 Out AttributeMathsOp_1 InA)
  (connect AddAttribute_2 Out AttributeMathsOp_1 InB)
  (connect DefaultInputNode In FilterDataByType_0 In)
  (connect FilterDataByType_0 InsideFilter ToPoint_2 In)
  (output Out AttributeMathsOp_1))
